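# Ambient spot sound cue for a power plant exhaust vent with rotating metal, heard outdoors.
# Mixes three looping layers, each picked randomly: HVAC or fan hum, rotating metal, and modulated duct air.
# A modulator and volume fade-in envelope shape the mix, with a random branch optionally adding a short start delay.
# Cone attenuation with HRTF, low-pass filtering, distance-based reverb send and limited concurrency localize it.
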SoundCue /WildEstate/Assets/Audio/POI/PowerPlant/Spots/SC_Spot_Exhaust_Vent_Rotating_Metal_Loop_Cone_Cue
root=QualityLevel_0
nodes [
 {id=Delay_0, children [Enveloper_1], properties {delayMax=0.3}}
 {
  id=Enveloper_1
  children [Modulator_0]
  properties {pitchCurve {editorCurveData {keys [{time=0}]}}, volumeCurve {editorCurveData {keys [<at 0 0>,<at 1 1>]}}}
 }
 {id=Mixer_1, children [Random_0,Random_2,Modulator_1], properties {inputVolume [2,0.5,1]}}
 {id=Mixer_2, children [WavePlayer_8], properties {inputVolume [3]}}
 {id=Mixer_3, children [WavePlayer_9], properties {inputVolume [2.5]}}
 {id=Mixer_4, children [WavePlayer_3], properties {inputVolume [2]}}
 {id=Mixer_5, children [WavePlayer_4], properties {inputVolume [3]}}
 {id=Mixer_6, children [WavePlayer_7], properties {inputVolume [0.5]}}
 {id=Mixer_7, children [WavePlayer_6], properties {inputVolume [0.5]}}
 {id=Mixer_8, children [WavePlayer_2], properties {inputVolume [1.5]}}
 {id=Mixer_9, children [WavePlayer_10], properties {inputVolume [1.5]}}
 {id=Modulator_0, children [Mixer_1], properties {pitchMax=1.1}}
 {id=Modulator_1, children [Random_4]}
 {id=QualityLevel_0, children [Random_1,Random_1,Random_1,Random_1]}
 {id=Random_0, children [WavePlayer_0,Mixer_8,WavePlayer_5,Mixer_6,Mixer_7,Mixer_9], properties {weights [1,1,1,1,1,1]}}
 {id=Random_1, children [Delay_0,Enveloper_1], properties {weights [1,1]}}
 {id=Random_2, children [Mixer_4,Mixer_5,Mixer_2,Mixer_3], properties {weights [1,1,1,1]}}
 {id=Random_4, children [WavePlayer_11, WavePlayer_12], properties {weights [1,1]}}
 {id=WavePlayer_0, properties {bLooping=true, soundWaveAssetPtr=/AmbientAudioContent/Waves/SpottedLoops/Spot_City_HVAC_Aircon_Unit_Loop_07a}}
 {id=WavePlayer_10, properties {bLooping=true, soundWaveAssetPtr=/AmbientAudioContent/Waves/SpottedLoops/Spot_AirborneFan_Blades_Close_Loop}}
 {id=WavePlayer_11, properties {bLooping=true, soundWaveAssetPtr=/AmbientAudioContent/Waves/SpottedLoops/Spot_AC_Duct_Air_03}}
 {id=WavePlayer_12, properties {bLooping=true, soundWaveAssetPtr=/AmbientAudioContent/Waves/SpottedLoops/Spot_AC_Duct_Air_02}}
 {id=WavePlayer_2, properties {bLooping=true, soundWaveAssetPtr=/AmbientAudioContent/Waves/SpottedLoops/Spot_City_HVAC_Aircon_Unit_Loop_02a}}
 {
  id=WavePlayer_3
  properties {bLooping=true, soundWaveAssetPtr=/AmbientAudioContent/Waves/SpottedLoops/Spot_ExhaustStackA_Rotating_Metal_Loop_01}
 }
 {
  id=WavePlayer_4
  properties {bLooping=true, soundWaveAssetPtr=/AmbientAudioContent/Waves/SpottedLoops/Spot_ExhaustStackA_Rotating_Metal_Loop_02}
 }
 {id=WavePlayer_5, properties {bLooping=true, soundWaveAssetPtr=/AmbientAudioContent/Waves/SpottedLoops/Spot_City_HVAC_Aircon_Unit_Loop_07b}}
 {id=WavePlayer_6, properties {bLooping=true, soundWaveAssetPtr=/AmbientAudioContent/Waves/SpottedLoops/Spot_City_HVAC_Aircon_Unit_Loop_08a}}
 {id=WavePlayer_7, properties {bLooping=true, soundWaveAssetPtr=/AmbientAudioContent/Waves/SpottedLoops/Spot_City_HVAC_Aircon_Unit_Loop_08b}}
 {
  id=WavePlayer_8
  properties {bLooping=true, soundWaveAssetPtr=/AmbientAudioContent/Waves/SpottedLoops/Spot_ExhaustStackA_Rotating_Metal_Loop_05}
 }
 {
  id=WavePlayer_9
  properties {bLooping=true, soundWaveAssetPtr=/AmbientAudioContent/Waves/SpottedLoops/Spot_ExhaustStackB_Rotating_Metal_Loop_01}
 }
]
properties {
 bEnableBusSends=false
 bOverrideAttenuation=true
 bOverrideConcurrency=true
 maxDistance=1200
 soundClassObject=&Fort_Ambient_Outdoors_Athena
 volumeMultiplier=0.18
 attenuationOverrides {
  attenuationShape=Cone
  bAttenuateWithLPF=true
  bEnableListenerFocus=true
  distanceAlgorithm=Logarithmic
  falloffDistance=950
  hPFFrequencyAtMax=100
  lPFFrequencyAtMax=2000
  lPFRadiusMax=850
  lPFRadiusMin=500
  nonFocusVolumeAttenuation=0.5
  spatializationAlgorithm=SPATIALIZATION_HRTF
  stereoSpread=100
  attenuationShapeExtents=<250 90>
 }
 concurrencyOverrides {maxCount=3, voiceStealReleaseTime=2, volumeScaleAttackTime=2, volumeScaleMode=Distance}
 concurrencySet [/AmbientAudioBase/Concurrency/AmbientConc_3Max_Spot_Default_Concurrency]
 soundSubmixSends [{
   maxSendDistance=750
   maxSendLevel=0.6
   minSendDistance=250
   minSendLevel=0.25
   soundSubmix=&Fortnite_Reverb_Default
   customSendLevelCurve {editorCurveData {keys [<at 0 0>,<at 1 1>]}}
  }]
}
