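# Ambient sound cue for an overhead console readout spot in the power plant POI, layering a readout bed with intermittent computer beeps and alarms.
# It starts immediately or after a one-second delay, behind a quality switch that fills only the first two slots.
# The bed randomly picks a looping wave under a looping volume envelope.
# Three separately attenuated loops add randomly delayed one-shots and alarm sounds, some pitch-modulated for variation.
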SoundCue /WildEstate/Assets/Audio/POI/PowerPlant/Spots/SC_Spot_Overhead_Console_Readout_Loop
root=Random_1
nodes [
 {
  id=Attenuation_0
  children [Looping_0]
  properties {attenuationSettings=/Game/Sounds/Artemis/Environment_Actors/Computer_Screen/Government_Computer_Smaller_Log_Att}
 }
 {
  id=Attenuation_1
  children [Looping_1]
  properties {attenuationSettings=&Government_Computer_Smaller_Log_Att}
 }
 {
  id=Attenuation_2
  children [Looping_2]
  properties {attenuationSettings=&Government_Computer_Smaller_Log_Att}
 }
 {id=Delay_0, children [QualityLevel_0], properties {delayMax=1, delayMin=1}}
 {id=Delay_1, children [Random_3], properties {delayMax=1, delayMin=0.5}}
 {id=Delay_2, children [Modulator_2], properties {delayMax=8, delayMin=3}}
 {id=Delay_3, children [Modulator_4], properties {delayMax=1, delayMin=0.5}}
 {
  id=Enveloper_0
  children [Modulator_0]
  properties {
   bLoop=true
   bLoopIndefinitely=true
   loopEnd=20
   pitchCurve {editorCurveData {keys [{time=0}]}}
   volumeCurve {editorCurveData {keys [<at -0.0028 1>,<at 0 1>,<at 10 0.95>,<at 20 1>]}}
  }
 }
 {id=Looping_0, children [Delay_1]}
 {id=Looping_1, children [Delay_2]}
 {id=Looping_2, children [Delay_3]}
 {id=Looping_3, children [Mixer_2], properties {bLoopIndefinitely=false, loopCount=2}}
 {id=Mixer_0, children [WavePlayer_19], properties {inputVolume [1]}}
 {id=Mixer_1, children [Mixer_5,Enveloper_0], properties {inputVolume [0.8,1.5]}}
 {id=Mixer_2, children [WavePlayer_17], properties {inputVolume [1]}}
 {id=Mixer_3, children [WavePlayer_6], properties {inputVolume [0.5]}}
 {id=Mixer_4, children [WavePlayer_7], properties {inputVolume [0.5]}}
 {id=Mixer_5, children [Attenuation_0, Attenuation_1, Attenuation_2], properties {inputVolume [0.6,0.2,0.7]}}
 {id=Mixer_6, children [WavePlayer_8], properties {inputVolume [0.5]}}
 {id=Mixer_7, children [WavePlayer_14], properties {inputVolume [0.75]}}
 {id=Modulator_0, children [Random_0], properties {pitchMax=1.1, volumeMax=1, volumeMin=0.9}}
 {id=Modulator_2, children [WavePlayer_4], properties {pitchMax=0.8, pitchMin=0.8, volumeMax=1, volumeMin=1}}
 {id=Modulator_3, children [Mixer_3], properties {pitchMax=1.15, volumeMax=1, volumeMin=1}}
 {id=Modulator_4, children [Random_5], properties {pitchMax=1.3, pitchMin=0.7, volumeMax=1, volumeMin=1}}
 {id=Modulator_5, children [Mixer_4], properties {pitchMax=1.15, volumeMax=1, volumeMin=1}}
 {id=Modulator_7, children [Mixer_6], properties {pitchMax=1.15, volumeMax=1, volumeMin=1}}
 {id=QualityLevel_0, children [Mixer_1,Mixer_1,null,null]}
 {id=Random_0, children [WavePlayer_1,WavePlayer_2], properties {weights [1,1]}}
 {id=Random_1, children [Delay_0, QualityLevel_0], properties {weights [1,1]}}
 {id=Random_3, children [Looping_3,Mixer_0,Modulator_3,Modulator_5,Modulator_7], properties {weights [1,1,1,1,1]}}
 {id=Random_5, children [WavePlayer_5, WavePlayer_9, WavePlayer_10, Mixer_7], properties {weights [1,1,1,1]}}
 {
  id=WavePlayer_1
  properties {bLooping=true, soundWaveAssetPtr=&/Spot_Overhead_Console_Readout_Loop_02}
 }
 {id=WavePlayer_10, properties {soundWaveAssetPtr=/AmbientAudioContent/Waves/OneShots/Spot_SecretComplex_Computer_Oneshot_12}}
 {id=WavePlayer_14, properties {soundWaveAssetPtr=/AmbientAudioContent/Waves/OneShots/Spot_SecretComplex_Computer_Oneshot_15}}
 {id=WavePlayer_17, properties {soundWaveAssetPtr=/AmbientAudioContent/Waves/SpottedLoops/Spot_Weapon_Lab_Alarm_Loop_01}}
 {id=WavePlayer_19, properties {soundWaveAssetPtr=/AmbientAudioContent/Waves/SpottedLoops/Spot_Weapon_Lab_Alarm_Loop_03}}
 {
  id=WavePlayer_2
  properties {bLooping=true, soundWaveAssetPtr=&/Spot_Overhead_Console_Readout_Loop_03}
 }
 {id=WavePlayer_4, properties {soundWaveAssetPtr=/AmbientAudioContent/Waves/OneShots/Spot_SecretComplex_Computer_Oneshot_01}}
 {id=WavePlayer_5, properties {soundWaveAssetPtr=/AmbientAudioContent/Waves/OneShots/Spot_SecretComplex_Computer_Oneshot_10}}
 {id=WavePlayer_6, properties {soundWaveAssetPtr=/AmbientAudioContent/Waves/OneShots/Spot_SecretComplex_Computer_Oneshot_03}}
 {id=WavePlayer_7, properties {soundWaveAssetPtr=/AmbientAudioContent/Waves/OneShots/Spot_SecretComplex_Computer_Oneshot_04}}
 {id=WavePlayer_8, properties {soundWaveAssetPtr=/AmbientAudioContent/Waves/OneShots/Spot_SecretComplex_Computer_Oneshot_05}}
 {id=WavePlayer_9, properties {soundWaveAssetPtr=/AmbientAudioContent/Waves/OneShots/Spot_SecretComplex_Computer_Oneshot_11}}
]
properties {
 attenuationSettings=/WildEstate/Assets/Audio/POI/Domestic_Commercial/Spots/Spot_WildEstate_Domestic_Commercial_Small_Att
 bDebug=true
 bOverrideConcurrency=true
 maxDistance=1100
 soundClassObject=&Fort_Ambient_Indoors_Athena
 volumeMultiplier=1
 attenuationOverrides {bSpatialize=false, stereoSpread=1147.9967, attenuationShapeExtents {x=6826.1992}}
 concurrencyOverrides {maxCount=2}
 concurrencySet [/AmbientAudioBase/Concurrency/AmbientConc_2Max_Loop_PreventNew_Default_Concurrency]
}
