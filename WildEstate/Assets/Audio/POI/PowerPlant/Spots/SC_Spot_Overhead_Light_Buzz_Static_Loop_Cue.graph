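# Sound cue for a looping overhead-light electrical buzz, an ambient spot sound in the power plant POI.
# Mixes a randomly picked, pitched-up hologram loop, weighted louder, with a randomly picked neon-light loop, each through its own modulator, then an overall one.
# An attenuation node applies cone falloff with HRTF, listener focus, distance filtering and occlusion.
# The quiet cue overrides concurrency to a low voice limit and sends to reverb.
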SoundCue /WildEstate/Assets/Audio/POI/PowerPlant/Spots/SC_Spot_Overhead_Light_Buzz_Static_Loop_Cue
root=Attenuation_5
nodes [
 {
  id=Attenuation_5
  children [Modulator_0]
  properties {
   bOverrideAttenuation=true
   attenuationOverrides {
    attenuationShape=Cone
    bAttenuateWithLPF=true
    bEnableFocusInterpolation=true
    bEnableListenerFocus=true
    bEnableLogFrequencyScaling=true
    bEnableOcclusion=true
    coneSphereFalloffDistance=250
    coneSphereRadius=50
    distanceAlgorithm=Logarithmic
    falloffDistance=500
    focusAzimuth=15
    focusDistanceScale=0.9
    focusReleaseInterpSpeed=0.8
    focusVolumeAttenuation=1.2
    hPFFrequencyAtMax=100
    hPFFrequencyAtMin=50
    lPFFrequencyAtMax=750
    lPFRadiusMax=1500
    lPFRadiusMin=300
    nonFocusAzimuth=45
    nonFocusVolumeAttenuation=0.2
    occlusionInterpolationTime=2
    occlusionLowPassFilterFrequency=1000
    occlusionTraceChannel=ECC_GameTraceChannel11
    occlusionVolumeAttenuation=0
    spatializationAlgorithm=SPATIALIZATION_HRTF
    stereoSpread=0
    attenuationShapeExtents=<500 60 22>
    customAttenuationCurve {editorCurveData {keys [
       {arriveTangent=-2.6039, interpMode=RCIM_Cubic, leaveTangent=-2.6039, tangentMode=RCTM_User, time=0, value=1}
       {arriveTangent=0.0377, interpMode=RCIM_Cubic, leaveTangent=0.0377, tangentMode=RCTM_User, time=0.5, value=0}
      ]}}
   }
  }
 }
 {id=Mixer_0, children [Modulator_1,Modulator_2], properties {inputVolume [2,1]}}
 {id=Modulator_0, children [Mixer_0], properties {pitchMin=1}}
 {id=Modulator_1, children [Random_2], properties {pitchMax=2, pitchMin=1.8}}
 {id=Modulator_2, children [Random_1], properties {pitchMax=1.2, pitchMin=1}}
 {id=Random_1, children [WavePlayer_5,WavePlayer_6], properties {weights [1,1]}}
 {id=Random_2, children [WavePlayer_0,WavePlayer_7,WavePlayer_8,WavePlayer_9], properties {weights [1,1,1,1]}}
 {id=WavePlayer_0, properties {bLooping=true, soundWaveAssetPtr=/AmbientAudioContent/Waves/SpottedLoops/Spot_Keys_And_Locks_HologramLoop_04}}
 {id=WavePlayer_5, properties {bLooping=true, soundWaveAssetPtr=/AmbientAudioContent/Waves/SpottedLoops/Spot_City_NeonLight_02}}
 {id=WavePlayer_6, properties {bLooping=true, soundWaveAssetPtr=/AmbientAudioContent/Waves/SpottedLoops/Spot_City_NeonLight_01}}
 {id=WavePlayer_7, properties {bLooping=true, soundWaveAssetPtr=/AmbientAudioContent/Waves/SpottedLoops/Spot_Keys_And_Locks_HologramLoop_03}}
 {id=WavePlayer_8, properties {bLooping=true, soundWaveAssetPtr=/AmbientAudioContent/Waves/SpottedLoops/Spot_Keys_And_Locks_HologramLoop_02}}
 {id=WavePlayer_9, properties {bLooping=true, soundWaveAssetPtr=/AmbientAudioContent/Waves/SpottedLoops/Spot_Keys_And_Locks_HologramLoop_01}}
]
properties {
 bOverrideConcurrency=true
 maxDistance=1000
 soundClassObject=&Fort_Ambient_Props_Athena
 volumeMultiplier=0.25
 attenuationOverrides {falloffDistance=2000}
 concurrencyOverrides {maxCount=2}
 concurrencySet [&Weapon_Drop_StopOldest]
 soundSubmixSends [{minSendLevel=0.25, sendLevelControlMethod=Linear, soundSubmix=&Fortnite_Reverb_Default}]
}
