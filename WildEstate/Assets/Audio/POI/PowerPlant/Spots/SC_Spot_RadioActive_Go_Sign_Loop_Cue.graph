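# Looping ambient SoundCue for the radioactive sign spot at the power plant POI, layering the sign loop with a quieter neon buzz.
# The mix passes through a modulator and two nested looping envelopers of different lengths, which slowly shape pitch and volume.
# Overrides give it short-range capsule attenuation with HRTF, distance filtering and listener focus.
# It also sends to distant reverb and the exterior ambient bus, with concurrency capped at two.
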SoundCue /WildEstate/Assets/Audio/POI/PowerPlant/Spots/SC_Spot_RadioActive_Go_Sign_Loop_Cue
root=Enveloper_2
nodes [
 {
  id=Enveloper_1
  children [Mixer_1]
  properties {
   bLoop=true
   bLoopIndefinitely=true
   loopEnd=30
   pitchCurve {editorCurveData {keys [{arriveTangent=-0.0013, interpMode=RCIM_Cubic, time=0}, <at 15 1.02 RCIM_Cubic>, {interpMode=RCIM_Cubic, leaveTangent=0.0014, time=30}]}}
   volumeCurve {editorCurveData {keys [<at 0 1>]}}
  }
 }
 {
  id=Enveloper_2
  children [Enveloper_1]
  properties {
   bLoop=true
   bLoopIndefinitely=true
   loopEnd=20
   pitchCurve {editorCurveData {keys [{time=0}]}}
   volumeCurve {editorCurveData {keys [
      {arriveTangent=-0.01, interpMode=RCIM_Cubic, time=0, value=1}
      <at 10 0.9 RCIM_Cubic>
      {interpMode=RCIM_Cubic, leaveTangent=0.01, time=20, value=1}
     ]}}
  }
 }
 {id=Mixer_0, children [WavePlayer_0, WavePlayer_13], properties {inputVolume [1,0.5]}}
 {id=Mixer_1, children [Modulator_3,null], properties {inputVolume [1,0.05]}}
 {id=Modulator_3, children [Mixer_0]}
 {id=WavePlayer_0, properties {bLooping=true, soundWaveAssetPtr=&/Spot_RadioActive_Go_Sign_Loop}}
 {id=WavePlayer_13, properties {bLooping=true, soundWaveAssetPtr=/AmbientAudioContent/Waves/SpottedLoops/spot_neon_sign_buzz_loop_01}}
]
properties {
 bDebug=true
 bOverrideAttenuation=true
 bOverrideConcurrency=true
 maxDistance=1000
 soundClassObject=&Fort_Ambient_Outdoors_Athena
 volumeMultiplier=1
 attenuationOverrides {
  attenuationShape=Capsule
  bAttenuateWithLPF=true
  bEnableListenerFocus=true
  bEnableLogFrequencyScaling=true
  bEnableSubmixSends=true
  distanceAlgorithm=Logarithmic
  falloffDistance=850
  hPFFrequencyAtMax=200
  lPFFrequencyAtMax=1000
  lPFRadiusMax=1000
  lPFRadiusMin=200
  nonFocusVolumeAttenuation=0.5
  spatializationAlgorithm=SPATIALIZATION_HRTF
  stereoSpread=0
  attenuationShapeExtents=<150 100>
  submixSendSettings [{
    bEnableHPFCutoff=false
    bEnableLPFCutoff=false
    disableManualSendClamp=false
    hPFCutoff=20
    lPFCutoff=20000
    maxSendDistance=1000
    maxSendLevel=0.25
    minSendDistance=0
    minSendLevel=0
    sendLevel=0.2
    sendLevelControlMethod=Linear
    soundSubmix=&Fortnite_Reverb_Distant
   }]
 }
 busSends [{audioBus=null, minSendLevel=0, sendLevel=1, soundSourceBus=&AmbientAudio_Exterior_SourceBus}]
 concurrencyOverrides {maxCount=2}
 concurrencySet [/AmbientAudioBase/Concurrency/AmbientConc_2Max_Loop_PreventNew_Default_Concurrency]
}
