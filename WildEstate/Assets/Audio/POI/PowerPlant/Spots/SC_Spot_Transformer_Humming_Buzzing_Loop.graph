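# Sound cue for a power plant transformer ambient spot, mixing a continuous humming, buzzing bed with intermittent electrical arcing.
# The bed randomly picks one of two looping hum waves, with slight volume modulation and an endlessly looping envelope that gently dips and recovers.
# A parallel looping branch waits a random delay, then plays one of six resonant arcing one-shots with volume variation.
# It borrows city HVAC loop attenuation and uses a two-max, prevent-new ambient concurrency.
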SoundCue /WildEstate/Assets/Audio/POI/PowerPlant/Spots/SC_Spot_Transformer_Humming_Buzzing_Loop
root=Mixer_0
nodes [
 {id=Delay_0, children [Random_3], properties {delayMax=25, delayMin=10}}
 {
  id=Enveloper_0
  children [Modulator_0]
  properties {
   bLoop=true
   bLoopIndefinitely=true
   loopEnd=20
   pitchCurve {editorCurveData {keys [{time=0}]}}
   volumeCurve {editorCurveData {keys [<at -0.0028 1>,<at 0 1>,<at 10 0.95>,<at 20 1>]}}
  }
 }
 {id=Looping_0, children [Delay_0]}
 {id=Mixer_0, children [Enveloper_0,Modulator_1], properties {inputVolume [1,1]}}
 {id=Modulator_0, children [Random_2], properties {volumeMax=1, volumeMin=0.9}}
 {id=Modulator_1, children [Looping_0], properties {volumeMax=1, volumeMin=0.9}}
 {id=Random_2, children [WavePlayer_5,WavePlayer_6], properties {weights [1,1]}}
 {
  id=Random_3
  children [WavePlayer_7, WavePlayer_8, WavePlayer_9, WavePlayer_10, WavePlayer_11, WavePlayer_12]
  properties {weights [1,1,1,1,1,1]}
 }
 {
  id=WavePlayer_10
  properties {soundWaveAssetPtr=&/Spot_Transformer_Electrical_Arcing_Resonant_04}
 }
 {
  id=WavePlayer_11
  properties {soundWaveAssetPtr=&/Spot_Transformer_Electrical_Arcing_Resonant_05}
 }
 {
  id=WavePlayer_12
  properties {soundWaveAssetPtr=&/Spot_Transformer_Electrical_Arcing_Resonant_06}
 }
 {
  id=WavePlayer_5
  properties {bLooping=true, soundWaveAssetPtr=&/Spot_Transformer_Humming_Buzzing_Deep_Presence_Loop_01}
 }
 {
  id=WavePlayer_6
  properties {bLooping=true, soundWaveAssetPtr=&/Spot_Transformer_Humming_Buzzing_Deep_Presence_Loop_02}
 }
 {id=WavePlayer_7, properties {soundWaveAssetPtr=&/Spot_Transformer_Electrical_Arcing_Resonant_01}}
 {id=WavePlayer_8, properties {soundWaveAssetPtr=&/Spot_Transformer_Electrical_Arcing_Resonant_02}}
 {id=WavePlayer_9, properties {soundWaveAssetPtr=&/Spot_Transformer_Electrical_Arcing_Resonant_03}}
]
properties {
 attenuationSettings=/AmbientAudioContent/SoundCues/SpottedLoops/ATT_Spot_City_HVAC_Aircon_Loop
 bDebug=true
 maxDistance=1560
 soundClassObject=&Fort_Ambient_Indoors_Athena
 volumeMultiplier=1
 attenuationOverrides {bSpatialize=false, stereoSpread=1147.9967, attenuationShapeExtents {x=6826.1992}}
 concurrencySet [/AmbientAudioBase/Concurrency/AmbientConc_2Max_Loop_PreventNew_Default_Concurrency]
}
