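# Looping ambient cue for the experiment machine at the Wild Estate power plant, mixing random beeps over a radiation bed.
# Single beeps and beep phrases repeat on separate randomly delayed loops.
# A random radiation loop, sometimes start-delayed, runs through nested looping envelopes that slowly drift volume and pitch.
# Each layer has its own HRTF attenuation with focus and occlusion, muffling the bed harder; every quality level shares one mix.
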SoundCue /WildEstate/Assets/Audio/POI/PowerPlant/Spots/SC_Spot_WildEstate_ExperimentMachine_Loop_Cue
root=QualityLevel_0
nodes [
 {
  id=Attenuation_0
  children [Mixer_1]
  properties {
   bOverrideAttenuation=true
   attenuationOverrides {
    bEnableListenerFocus=true
    bEnableOcclusion=true
    distanceAlgorithm=Custom
    falloffDistance=1000
    nonFocusVolumeAttenuation=0.75
    occlusionLowPassFilterFrequency=850
    occlusionTraceChannel=ECC_GameTraceChannel11
    occlusionVolumeAttenuation=0.5
    spatializationAlgorithm=SPATIALIZATION_HRTF
    stereoSpread=0
    attenuationShapeExtents {x=0}
    customAttenuationCurve {editorCurveData {keys [
       <at 0 1>
       {arriveTangent=-1, interpMode=RCIM_Cubic, leaveTangent=-1, time=0.1318, value=0.4839}
       {arriveTangent=-2.0945, interpMode=RCIM_Cubic, leaveTangent=-2.0945, tangentMode=RCTM_User, time=1, value=0}
      ]}}
   }
  }
 }
 {
  id=Attenuation_1
  children [Enveloper_1]
  properties {
   bOverrideAttenuation=true
   attenuationOverrides {
    bEnableListenerFocus=true
    bEnableOcclusion=true
    distanceAlgorithm=Custom
    falloffDistance=1000
    nonFocusVolumeAttenuation=0.75
    occlusionLowPassFilterFrequency=100
    occlusionTraceChannel=ECC_GameTraceChannel11
    occlusionVolumeAttenuation=0
    spatializationAlgorithm=SPATIALIZATION_HRTF
    stereoSpread=0
    attenuationShapeExtents {x=0}
    customAttenuationCurve {editorCurveData {keys [
       <at 0 1>
       <at 0.08 1>
       {arriveTangent=-1, interpMode=RCIM_Cubic, leaveTangent=-1, time=0.2788, value=0.4961}
       {arriveTangent=-2.0945, interpMode=RCIM_Cubic, leaveTangent=-2.0945, tangentMode=RCTM_User, time=1, value=0}
      ]}}
   }
  }
 }
 {id=Delay_0, children [Modulator_1], properties {delayMax=3, delayMin=0.25}}
 {id=Delay_1, children [Random_1], properties {delayMax=0.25, delayMin=0.25}}
 {id=Delay_2, children [Modulator_2], properties {delayMax=5, delayMin=3}}
 {
  id=Enveloper_0
  children [Modulator_0]
  properties {
   bLoop=true
   bLoopIndefinitely=true
   loopEnd=60
   pitchCurve {editorCurveData {keys [{interpMode=RCIM_Cubic, time=0}, <at 30 0.99 RCIM_Cubic>, {time=60}]}}
   volumeCurve {editorCurveData {keys [<at 0 1>]}}
  }
 }
 {
  id=Enveloper_1
  children [Enveloper_0]
  properties {
   bLoop=true
   bLoopIndefinitely=true
   loopEnd=60
   pitchCurve {editorCurveData {keys [{time=0}]}}
   volumeCurve {editorCurveData {keys [
      {arriveTangent=-0.01, interpMode=RCIM_Cubic, time=0, value=1}
      {arriveTangent=-0.0035, interpMode=RCIM_Cubic, leaveTangent=-0.0035, tangentMode=RCTM_User, time=4.5428, value=0.8801}
      {arriveTangent=0.0016, interpMode=RCIM_Cubic, leaveTangent=0.0016, tangentMode=RCTM_User, time=7.4135, value=0.9776}
      {arriveTangent=0.0008, interpMode=RCIM_Cubic, leaveTangent=0.0008, tangentMode=RCTM_User, time=11.1077, value=0.8936}
      {arriveTangent=0.0016, interpMode=RCIM_Cubic, leaveTangent=0.0016, tangentMode=RCTM_User, time=12.9775, value=0.8792}
      {arriveTangent=0.0009, interpMode=RCIM_Cubic, leaveTangent=0.0009, time=14.965, value=0.9463}
      {arriveTangent=-0.0011, interpMode=RCIM_Cubic, leaveTangent=-0.0011, tangentMode=RCTM_User, time=18.5868, value=0.8842}
      {arriveTangent=0.0002, interpMode=RCIM_Cubic, leaveTangent=0.0002, time=23.9559, value=0.9962}
      {arriveTangent=0.0016, interpMode=RCIM_Cubic, leaveTangent=0.0016, tangentMode=RCTM_User, time=28.034, value=0.8864}
      {arriveTangent=-0.0003, interpMode=RCIM_Cubic, leaveTangent=-0.0003, time=31.2025, value=0.961}
      {arriveTangent=0.0009, interpMode=RCIM_Cubic, leaveTangent=0.0009, tangentMode=RCTM_User, time=34.9227, value=0.8844}
      {arriveTangent=-0.0008, interpMode=RCIM_Cubic, leaveTangent=-0.0008, time=37.93, value=0.9962}
      <at 42.0081 0.879>
      {arriveTangent=0.0024, interpMode=RCIM_Cubic, leaveTangent=0.0024, tangentMode=RCTM_User, time=44.5668, value=0.8842}
      {arriveTangent=-0.0006, interpMode=RCIM_Cubic, leaveTangent=-0.0006, time=47.958, value=0.9776}
      {arriveTangent=-0.0056, interpMode=RCIM_Cubic, leaveTangent=-0.0056, tangentMode=RCTM_User, time=53.1343, value=0.8792}
      <at 55.3163 0.9711>
      {arriveTangent=-0.0003, interpMode=RCIM_Cubic, leaveTangent=-0.0003, tangentMode=RCTM_User, time=57.7536, value=0.8789}
      <at 60 1 RCIM_Cubic>
     ]}}
  }
 }
 {id=Looping_0, children [Delay_0]}
 {id=Looping_1, children [Delay_2]}
 {id=Mixer_0, children [Attenuation_0, Attenuation_1], properties {inputVolume [1,1.5]}}
 {id=Mixer_1, children [Looping_0,Looping_1], properties {inputVolume [0.8,1]}}
 {id=Modulator_0, children [Random_0], properties {pitchMax=1.01, pitchMin=0.99}}
 {id=Modulator_1, children [Random_5], properties {pitchMax=1.1, pitchMin=0.9}}
 {id=Modulator_2, children [Random_4]}
 {id=QualityLevel_0, children [Mixer_0,Mixer_0,Mixer_0,Mixer_0]}
 {id=Random_0, children [Delay_1,Random_1], properties {weights [1,1]}}
 {id=Random_1, children [WavePlayer_0,WavePlayer_1,WavePlayer_2], properties {weights [1,1,1]}}
 {
  id=Random_4
  children [WavePlayer_20, WavePlayer_21, WavePlayer_22, WavePlayer_23, WavePlayer_24, WavePlayer_25]
  properties {weights [1,1,1,1,1,1]}
 }
 {
  id=Random_5
  children [WavePlayer_26, WavePlayer_27, WavePlayer_28, WavePlayer_29, WavePlayer_30, WavePlayer_31, WavePlayer_32, WavePlayer_34]
  properties {weights [1,1,1,1,1,1,1,1]}
 }
 {
  id=WavePlayer_0
  properties {bLooping=true, soundWaveAssetPtr=&/Spot_WildEstate_Radiation_Loop_01}
 }
 {
  id=WavePlayer_1
  properties {bLooping=true, soundWaveAssetPtr=&/Spot_WildEstate_Radiation_Loop_02}
 }
 {
  id=WavePlayer_2
  properties {bLooping=true, soundWaveAssetPtr=&/Spot_WildEstate_Radiation_Loop_03}
 }
 {
  id=WavePlayer_20
  properties {soundWaveAssetPtr=&/SC_Spot_WildEstate_ExperimentMachine_Beep_Phrase_01}
 }
 {
  id=WavePlayer_21
  properties {soundWaveAssetPtr=&/SC_Spot_WildEstate_ExperimentMachine_Beep_Phrase_02}
 }
 {
  id=WavePlayer_22
  properties {soundWaveAssetPtr=&/SC_Spot_WildEstate_ExperimentMachine_Beep_Phrase_03}
 }
 {
  id=WavePlayer_23
  properties {soundWaveAssetPtr=&/SC_Spot_WildEstate_ExperimentMachine_Beep_Phrase_04}
 }
 {
  id=WavePlayer_24
  properties {soundWaveAssetPtr=&/SC_Spot_WildEstate_ExperimentMachine_Beep_Phrase_05}
 }
 {
  id=WavePlayer_25
  properties {soundWaveAssetPtr=&/SC_Spot_WildEstate_ExperimentMachine_Beep_Phrase_06}
 }
 {
  id=WavePlayer_26
  properties {soundWaveAssetPtr=&/SC_Spot_WildEstate_ExperimentMachine_Beep_Single_01}
 }
 {
  id=WavePlayer_27
  properties {soundWaveAssetPtr=&/SC_Spot_WildEstate_ExperimentMachine_Beep_Single_02}
 }
 {
  id=WavePlayer_28
  properties {soundWaveAssetPtr=&/SC_Spot_WildEstate_ExperimentMachine_Beep_Single_03}
 }
 {
  id=WavePlayer_29
  properties {soundWaveAssetPtr=&/SC_Spot_WildEstate_ExperimentMachine_Beep_Single_04}
 }
 {
  id=WavePlayer_30
  properties {soundWaveAssetPtr=&/SC_Spot_WildEstate_ExperimentMachine_Beep_Single_05}
 }
 {
  id=WavePlayer_31
  properties {soundWaveAssetPtr=&/SC_Spot_WildEstate_ExperimentMachine_Beep_Single_06}
 }
 {
  id=WavePlayer_32
  properties {soundWaveAssetPtr=&/SC_Spot_WildEstate_ExperimentMachine_Beep_Single_07}
 }
 {
  id=WavePlayer_34
  properties {soundWaveAssetPtr=&/SC_Spot_WildEstate_ExperimentMachine_Beep_Single_09}
 }
]
properties {
 bOverrideConcurrency=true
 maxDistance=1000
 soundClassObject=&Fort_Ambient_Props_Athena
 volumeMultiplier=0.7
 attenuationOverrides {bEnableOcclusion=true, falloffDistance=850, nonSpatializedRadiusStart=700, stereoSpread=1000, attenuationShapeExtents {x=250}}
 concurrencyOverrides {maxCount=3, voiceStealReleaseTime=2}
 soundSubmixSends [{maxSendLevel=1.2, minSendLevel=1, sendLevel=1.2, soundSubmix=&Fortnite_Reverb_Default}]
}
