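# Looping ambient cue for the power plant cooling tower's presence, layered by listener distance in the Athena ambient sound class.
# A mixer blends close, medium and distant looping waves, each with its own attenuation node, weighted toward the close layer.
# The close layer is a non-spatialized, occluded capsule.
# The outer layers use listener focus, with occlusion and random pitch modulation only on the distant one.
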SoundCue /WildEstate/Assets/Audio/POI/PowerPlant/Spots/Spot_PowerPlant_CoolingTower_Pressence_Loop_Cue
root=Mixer_0
nodes [
 {
  id=Attenuation_0
  children [WavePlayer_3]
  properties {
   bOverrideAttenuation=true
   attenuationOverrides {
    attenuationShape=Capsule
    bEnableOcclusion=true
    bSpatialize=false
    distanceAlgorithm=Logarithmic
    falloffDistance=1400
    occlusionLowPassFilterFrequency=1000
    reverbWetLevelMin=0.6429
    attenuationShapeExtents=<5802.106 2761.6406>
   }
  }
 }
 {
  id=Attenuation_1
  children [WavePlayer_5]
  properties {
   attenuationSettings=&Fort_Weapon_Projectile_Medium_Mono_Athena_Att
   bOverrideAttenuation=true
   attenuationOverrides {
    bEnableListenerFocus=true
    falloffDistance=4033.6787
    focusAzimuth=55
    nonFocusVolumeAttenuation=0.5
    stereoSpread=5503.8877
    attenuationShapeExtents {x=3703.1514}
   }
  }
 }
 {
  id=Attenuation_2
  children [Modulator_0]
  properties {
   attenuationSettings=&Fort_Weapon_Projectile_Distant_Mono_Athena_Att
   bOverrideAttenuation=true
   attenuationOverrides {
    bEnableListenerFocus=true
    bEnableOcclusion=true
    falloffDistance=16663.5703
    focusAzimuth=50
    nonFocusAzimuth=70
    nonFocusVolumeAttenuation=0.5
    occlusionInterpolationTime=1
    occlusionLowPassFilterFrequency=500
    attenuationShapeExtents {x=9252.1387}
   }
  }
 }
 {id=Mixer_0, children [Attenuation_0, Attenuation_1, Attenuation_2], properties {inputVolume [1.4,1.2,1]}}
 {id=Modulator_0, children [WavePlayer_4], properties {pitchMax=1.2, pitchMin=0.8}}
 {
  id=WavePlayer_3
  properties {bLooping=true, soundWaveAssetPtr=&/Spot_PowerPlant_CoolingTower_Pressence_Close}
 }
 {
  id=WavePlayer_4
  properties {bLooping=true, soundWaveAssetPtr=&/Spot_PowerPlant_CoolingTower_Pressence_Distant}
 }
 {
  id=WavePlayer_5
  properties {bLooping=true, soundWaveAssetPtr=&/Spot_PowerPlant_CoolingTower_Pressence_Medium_Dist}
 }
]
properties {
 bDebug=true
 maxDistance=25915.709
 soundClassObject=&Fort_Ambient_Athena
 volumeMultiplier=0.33
 attenuationOverrides {bSpatialize=false, stereoSpread=1147.9967, attenuationShapeExtents {x=6826.1992}}
}
