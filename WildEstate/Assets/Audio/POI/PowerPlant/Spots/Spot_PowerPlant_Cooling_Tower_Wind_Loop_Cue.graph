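# Looping ambient wind cue for the Power Plant cooling tower, on an outdoor-ducking ambient sound class with a small concurrency limit.
# A mixer blends three capsule-attenuated branches: a base wind loop and a vortex loop, both non-spatialized with logarithmic falloff, plus a spatialized branch sharing the vortex player.
# That third branch uses a custom curve peaking at mid falloff and fading at both ends, with low-pass filtering and listener focus.
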SoundCue /WildEstate/Assets/Audio/POI/PowerPlant/Spots/Spot_PowerPlant_Cooling_Tower_Wind_Loop_Cue
root=Mixer_0
nodes [
 {
  id=Attenuation_1
  children [WavePlayer_0]
  properties {
   bOverrideAttenuation=true
   attenuationOverrides {
    attenuationShape=Capsule
    bSpatialize=false
    distanceAlgorithm=Logarithmic
    falloffDistance=2500
    reverbWetLevelMin=0.6429
    attenuationShapeExtents=<9126 2500>
   }
  }
 }
 {
  id=Attenuation_2
  children [WavePlayer_0]
  properties {
   bOverrideAttenuation=true
   attenuationOverrides {
    attenuationShape=Capsule
    bAttenuateWithLPF=true
    bEnableFocusInterpolation=true
    bEnableListenerFocus=true
    bEnableLogFrequencyScaling=true
    distanceAlgorithm=Custom
    falloffDistance=5000
    focusAttackInterpSpeed=0.1
    focusReleaseInterpSpeed=0.1
    hPFFrequencyAtMax=350
    lPFFrequencyAtMax=1000
    lPFRadiusMax=4000
    nonFocusVolumeAttenuation=0.4
    reverbWetLevelMin=0.6429
    stereoSpread=300
    attenuationShapeExtents {x=9126}
    customAttenuationCurve {editorCurveData {keys [
       {arriveTangent=-0.6384, interpMode=RCIM_Cubic, time=0.45, value=0}
       {arriveTangent=0.007, interpMode=RCIM_Cubic, leaveTangent=0.007, tangentMode=RCTM_User, time=0.65, value=0.8262}
       {arriveTangent=-2.0706, interpMode=RCIM_Cubic, leaveTangent=-2.0706, tangentMode=RCTM_User, time=0.7466, value=0.6688}
       {interpMode=RCIM_Cubic, tangentMode=RCTM_User, time=1.1, value=0}
      ]}}
   }
  }
 }
 {
  id=Attenuation_5
  children [WavePlayer_2]
  properties {
   bOverrideAttenuation=true
   attenuationOverrides {
    attenuationShape=Capsule
    bSpatialize=false
    distanceAlgorithm=Logarithmic
    falloffDistance=2500
    reverbWetLevelMin=0.6429
    attenuationShapeExtents=<9126 2500>
   }
  }
 }
 {id=Mixer_0, children [Attenuation_5, Attenuation_1, Attenuation_2], properties {inputVolume [1,0.66,0.5]}}
 {
  id=WavePlayer_0
  properties {bLooping=true, soundWaveAssetPtr=&/Spot_PowerPlant_Cooling_Tower_Wind_Vortex_Loop}
 }
 {id=WavePlayer_2, properties {bLooping=true, soundWaveAssetPtr=/Game/Sounds/Fort_Prop_Sounds/Exterior/CoolingTower/CoolingTowerWind_01}}
]
properties {
 bDebug=true
 bOverrideConcurrency=true
 maxDistance=14626
 soundClassObject=/Game/Sounds/FortSoundClassesAndModes/Fort_Ambient_DuckOutdoors_Athena_PSM
 volumeMultiplier=1
 concurrencyOverrides {maxCount=2}
}
