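# Looping exterior spot ambience for the WildEstate power plant's nuclear core, routed to the outdoor ambient sound class.
# A Mixer layers four looping hums and flows, including a pitched-down hover-tractor loop, into a shared Modulator.
# A Random node picks either that Modulator directly or the same chain through a short delay and fade-in Enveloper.
# Overrides add capsule attenuation with distance low-pass and listener focus, and cap concurrency.
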SoundCue /WildEstate/Assets/Audio/POI/PowerPlant/Spots/Spot_WildEstate_PowerPlant_Nuclear_Core_Loop_02_Ext_Cue
root=Random_0
nodes [
 {id=Delay_0, children [Enveloper_0], properties {delayMax=0.3, delayMin=0.3}}
 {
  id=Enveloper_0
  children [Modulator_1]
  properties {pitchCurve {editorCurveData {keys [{time=0}]}}, volumeCurve {editorCurveData {keys [<at 0 0>,<at 0.5 1>]}}}
 }
 {id=Mixer_0, children [Modulator_0,WavePlayer_2,WavePlayer_5,WavePlayer_6], properties {inputVolume [0.5,1,3,3]}}
 {id=Modulator_0, children [WavePlayer_0], properties {pitchMax=0.6, pitchMin=0.6}}
 {id=Modulator_1, children [Mixer_0]}
 {id=Random_0, children [Delay_0,Modulator_1], properties {weights [1,1]}}
 {id=WavePlayer_0, properties {bLooping=true, soundWaveAssetPtr=/AmbientAudioContent/Waves/SpottedLoops/Spot_Hover_Tractor_Loop_01}}
 {id=WavePlayer_2, properties {bLooping=true, soundWaveAssetPtr=/AmbientAudioContent/Waves/SpottedLoops/Spot_Hologram_Giant_Hum_Buzz_Loop_01}}
 {
  id=WavePlayer_5
  properties {bLooping=true, soundWaveAssetPtr=/AmbientAudioContent/Waves/SpottedLoops/Spot_NitroSilo_Liquid_Metal_Flowing_Inside_Loop}
 }
 {id=WavePlayer_6, properties {bLooping=true, soundWaveAssetPtr=&/Spot_RadioActive_Go_Sign_Loop}}
]
properties {
 bEnableBusSends=false
 bOverrideAttenuation=true
 bOverrideConcurrency=true
 maxDistance=1300
 soundClassObject=&Fort_Ambient_Outdoors_Athena
 volumeMultiplier=1.5
 attenuationOverrides {
  attenuationShape=Capsule
  bAttenuateWithLPF=true
  bEnableListenerFocus=true
  bEnableLogFrequencyScaling=true
  falloffDistance=850
  lPFFrequencyAtMax=2000
  lPFRadiusMax=1000
  lPFRadiusMin=650
  nonFocusVolumeAttenuation=0.75
  occlusionLowPassFilterFrequency=850
  occlusionTraceChannel=ECC_GameTraceChannel6
  occlusionVolumeAttenuation=0.5
  attenuationShapeExtents=<450 300>
 }
 concurrencyOverrides {maxCount=2}
}
