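# Looping ambient spot sound cue for the nuclear core at the WildEstate power plant POI, classed as an ambient prop.
# A mixer layers three looped waves: a pitched-down hover tractor, a heavily boosted rock hum and flowing liquid metal, all under a root modulator.
# Capsule attenuation override adds distance low-pass filtering and listener focus, with low cue volume and a small concurrency cap.
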SoundCue /WildEstate/Assets/Audio/POI/PowerPlant/Spots/Spot_WildEstate_PowerPlant_Nuclear_Core_Loop_03_Cue
root=Modulator_1
nodes [
 {id=Mixer_0, children [Modulator_0,WavePlayer_4,WavePlayer_1], properties {inputVolume [0.5,22,2]}}
 {id=Modulator_0, children [WavePlayer_0], properties {pitchMax=0.6, pitchMin=0.6}}
 {id=Modulator_1, children [Mixer_0]}
 {id=WavePlayer_0, properties {bLooping=true, soundWaveAssetPtr=/AmbientAudioContent/Waves/SpottedLoops/Spot_Hover_Tractor_Loop_01}}
 {
  id=WavePlayer_1
  properties {bLooping=true, soundWaveAssetPtr=/AmbientAudioContent/Waves/SpottedLoops/Spot_NitroSilo_Liquid_Metal_Flowing_Inside_Loop}
 }
 {
  id=WavePlayer_4
  properties {bLooping=true, soundWaveAssetPtr=/AmbientAudioContent/Waves/SpottedLoops/Spot_Amb_Tower_Aftermath_Humming_Rocks_Loop}
 }
]
properties {
 bEnableBusSends=false
 bOverrideAttenuation=true
 bOverrideConcurrency=true
 maxDistance=1300
 soundClassObject=&Fort_Ambient_Props_Athena
 volumeMultiplier=0.2
 attenuationOverrides {
  attenuationShape=Capsule
  bAttenuateWithLPF=true
  bEnableListenerFocus=true
  bEnableLogFrequencyScaling=true
  falloffDistance=850
  lPFFrequencyAtMax=1000
  lPFRadiusMax=1000
  lPFRadiusMin=500
  nonFocusVolumeAttenuation=0.75
  attenuationShapeExtents=<450 300>
 }
 concurrencyOverrides {maxCount=2}
}
