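# Sound Cue for a looping radioactive ambience spot at the WildEstate power plant point of interest.
# A looping Wave Player feeds a Modulator at its default settings as the root.
# Output is turned down and routed to the outdoor ambient sound class, with bus and submix sends disabled.
# A shared wide tile-ambient attenuation asset with air absorption handles distance falloff.
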SoundCue /WildEstate/Assets/Audio/POI/PowerPlant/Spots/Spot_WildEstate_PowerPlant_RadioActive_Loop_01_Cue
root=Modulator_1
nodes [
 {id=Modulator_1, children [WavePlayer_0]}
 {
  id=WavePlayer_0
  properties {bLooping=true, soundWaveAssetPtr=&/Spot_WildEstate_PowerPlant_RadioActive_Loop_01}
 }
]
properties {
 attenuationSettings=/Game/Sounds/Attenuation/Fort_TileAmbient_OneShot_Wider_AirAbsorbtion_Att
 bEnableBusSends=false
 bEnableSubmixSends=false
 maxDistance=8000
 soundClassObject=&Fort_Ambient_Outdoors_Athena
 volumeMultiplier=0.251
 attenuationOverrides {falloffDistance=6000, stereoSpread=10000, attenuationShapeExtents {x=2000}}
}
